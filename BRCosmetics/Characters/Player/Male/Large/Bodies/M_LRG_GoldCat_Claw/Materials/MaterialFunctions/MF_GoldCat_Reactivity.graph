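# Material function for the GoldCat player body that turns it gold as a super-levelling effect.
# It builds metallic gold attributes with tinted, desaturated detail and blends them over the incoming material using a mask gated by SL_CurrentLevel.
# MF_SuperLevellingMask_GoldCat derives level and emissive masks from pre-skinned position and normal, computed per vertex and interpolated.
# A masked emissive transition glow fades over a set duration, timed from a Blueprint-supplied game time.
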MaterialFunction /BRCosmetics/Characters/Player/Male/Large/Bodies/M_LRG_GoldCat_Claw/Materials/MaterialFunctions/MF_GoldCat_Reactivity
graph {nodes [
  Constant() {r=1}
  n1 = ComponentMask(None=NamedRerouteUsage()) {g=true}
  n2 = ScalarParameter() {defaultValue=0.3, parameterName=GMapIntensity}
  n3 = LinearInterpolate(A=Saturate(None=Add(A=@n1, B=@n2)), Alpha=NamedRerouteUsage()) {constB=0.5}
  n4 = NamedRerouteUsage()
  n5 = LinearInterpolate(Alpha=Desaturation(None=@n4, Fraction=Constant() {r=1})) {constA=0.4}
  n6 = ComponentMask(None=@n4) {a=true}
  n7 = Constant() {r=1}
  n8 = Constant()
  n9 = SetMaterialAttributes!(
   MaterialAttributes=NamedRerouteUsage()
   BaseColor=Multiply(
    A=LinearInterpolate(A=@n3, B=@n5, Alpha=@n6)
    B=VectorParameter() {parameterName=GoldBaseColor, defaultValue=<rgba 0.9047 0.5583 0.1384 1>}.RGB
   )
   Specular=@n7
   Roughness=LinearInterpolate(Alpha=NamedRerouteUsage()) {constA=0.25, constB=0.1}
   Metallic=@n7
   EmissiveColor=@n8
   Normal
   Opacity=@n8
  ) {attributeSetTypes [
    #1
    #2
    #3
    #4
    #5
    #6
    #7
   ]}
  n10 = NamedRerouteUsage()
  n11 = Constant()
  n12 = If(
   A=ScalarParameter() {defaultValue=100, group=SuperLevelling_Mask, parameterName=SL_CurrentLevel, sortPriority=3}
   "A > B"=Constant() {r=1}
   "A == B"=@n11
   "A < B"=@n11
  )
  n13 = GetMaterialAttributes(None=BlendMaterialAttributes(A=NamedRerouteUsage(), B=@n9, Alpha=Multiply(A=@n10, B=@n12))) {attributeGetTypes [#5]}
  n14 = NamedRerouteUsage()
  n15 = VectorParameter() {group=SuperLevelling_Mask, parameterName=Transition_EmissiveColor, sortPriority=101, defaultValue=<rgba 20 8.8102 1 1>}
  n16 = Subtract(A=Time(), B=ScalarParameter() {defaultValue=-99999, group=SuperLevelling_Mask, parameterName=SL_GameTimeFromBP, sortPriority=100})
  n17 = Add(A=ScalarParameter() {defaultValue=2, group=SuperLevelling_Mask, parameterName=SL_Animation_Duration, sortPriority=101}) {constB=0.1}
  n18 = OneMinus(None=Saturate(None=Divide(A=@n16, B=@n17)))
  n19 = Add(A=@n14, B=Multiply(A=@n15.RGB, B=SmoothStep(Value=@n18) {constMax=0.2, constValue=1}))
  n20 = NamedRerouteUsage()
  FunctionOutput(None=SetMaterialAttributes!(MaterialAttributes=@n13.MaterialAttributes, EmissiveColor=LinearInterpolate(A=@n14, B=@n19, Alpha=@n20)) {attributeSetTypes [#5]})
  n21 = LocalPosition() {localOrigin=InstancePreSkinning}
  n22 = PreSkinnedNormal()
  n23 = MaterialFunctionCall!(
   Local Position=ShaderStageSwitch(PixelShader=VertexInterpolator(VS=@n21.XYZ).PS, VertexShader=@n21.XYZ)
   Local Normal=ShaderStageSwitch(PixelShader=VertexInterpolator(VS=@n22).PS, VertexShader=@n22)
  ) {materialFunction=&/MF_SuperLevellingMask_GoldCat}
  NamedRerouteDeclaration(None=Saturate(None=ComponentMask(None=@n23.SuperLevelling_Mask) {r=true})) {name=SupeLevel_Mask, nodeColor=<rgba 0 0.0422 1 1>}
  n24 = FunctionInput() {bUsePreviewValueAsDefault=true, inputName=MouthFlipbook, inputType=FunctionInput_Vector4, sortPriority=10, previewValue=<1 1 1 1>}
  NamedRerouteDeclaration(None=@n24) {name=InMouthFlipbook, nodeColor=<rgba 0.1953 1 0 1>}
  n25 = FunctionInput() {bUsePreviewValueAsDefault=true, inputName=InTattooMask, inputType=FunctionInput_Scalar, sortPriority=10, previewValue=<0 0 0 0>}
  NamedRerouteDeclaration(None=@n25) {name=InTattooMask, nodeColor=<rgba 0 1 0.5078 1>}
  NamedRerouteDeclaration(None=Saturate(None=ComponentMask(None=@n23.SuperLevelling_Mask) {g=true})) {name=SuperLevel_Mask_Emissive, nodeColor=<rgba 0 0.1562 1 1>}
  NamedRerouteDeclaration(None=@n13.EmissiveColor) {name=CurrentEmissive, nodeColor=<rgba 0.9922 1 0 1>}
  NamedRerouteDeclaration(None=FunctionInput() {inputName=MaterialAttributes, inputType=FunctionInput_MaterialAttributes, sortPriority=10}) {name=InMaterialAttributes, nodeColor=<rgba 1 0.7266 0 1>}
  NamedRerouteDeclaration(None=FunctionInput() {bUsePreviewValueAsDefault=true, inputName=Mask, sortPriority=10, previewValue=<1 1 1 1>}) {name=inMask, nodeColor=<rgba 1 0.1875 0 1>}
 ]}
