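# Parent character material for the GoldCat Claw large male body cosmetic, built between the shared 2023 character PreFX and PostFX functions.
# An optional switch blends a tattoo color into base color, using a second-UV texture multiplied by a channel-selected mask.
# Another switch adds vertex-color flipbook mouths.
# The resulting attributes, tattoo mask and mouth color feed a GoldCat reactivity function, with named reroutes linking the stages.
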Material /BRCosmetics/Characters/Player/Male/Large/Bodies/M_LRG_GoldCat_Claw/Materials/M_GoldCat_Claw_Parent
graph {
 properties {
  bUseMaterialAttributes=true
  bUsedWithMeshDeformer=true
  bUsedWithMorphTargets=true
  bUsedWithSkeletalMesh=true
  pixelDepthOffsetMode=PDOM_Legacy
  shadingModel=MSM_FromMaterialExpression
  usedShadingModels=Default Lit | Subsurface | Clear Coat | Subsurface Profile | Cloth
 }
 nodes [
  n1 = MaterialFunctionCall() {materialFunction=&MF_CharacterParent_2023_PreFX}
  MaterialFunctionCall(Material=NamedRerouteUsage(), M=@n1.M, Pre Skinned Local Position=@n1.M, Distance Blend=@n1.M, Fresnel=@n1.M) {materialFunction=&MF_CharacterParent_2023_PostFX}
  NamedRerouteDeclaration(None=@n1.Result) {name=InMaterialAttributes, nodeColor=<rgba 0 0.2266 1 1>}
  n2 = GetMaterialAttributes(None=NamedRerouteUsage()) {attributeGetTypes [#1]}
  n3 = VectorParameter() {group=GoldCatClaw, parameterName=GC_TattooColor, defaultValue=<rgba 0.0742 0.0648 0.0545 1>}
  n4 = TextureSample() {
   constCoordinate=1
   samplerType=SAMPLERTYPE_Masks
   texture=/BRCosmetics/Characters/Player/Male/Large/Bodies/M_LRG_GoldCat_Claw/Textures/goldCatClaw_tattoo_texTransfer
  }
  n5 = ChannelMaskParameter(None=NamedRerouteUsage()) {parameterName=GC_TattooMask}
  n6 = NamedRerouteDeclaration(None=StaticSwitchParameter(True=Multiply(A=@n4.G, B=@n5.R), False=Constant()) {group=GoldCatClaw, parameterName=UseUV2Tattoos}) {name=TattooMask, nodeColor=<rgba 1 0 0.5391 1>}
  n7 = StaticSwitchParameter(
   True=SetMaterialAttributes!(MaterialAttributes=NamedRerouteUsage(), Base Color=LinearInterpolate(A=@n2.Base Color, B=@n3.RGB, Alpha=@n6)) {attributeSetTypes [#1]}
   False=NamedRerouteUsage()
  ) {group=GoldCatClaw, parameterName=UseUV2Tattoos}
  n8 = MaterialFunctionCall!(Material=@n7, MouthUVs) {materialFunction=/Game/Characters/Player/Common/MaterialFunctions/VertexColorFaces/MF_FV2_FlipbookMouths}
  n9 = AppendVector(A=@n8.Color, B=@n8.Color)
  n10 = Constant()
  NamedRerouteDeclaration(None=MaterialFunctionCall!(
    Mask=NamedRerouteUsage()
    MouthFlipbook=StaticSwitchParameter(True=@n9, False=@n10) {group=GoldCatClaw, parameterName=UseFlipbookMouth}
    InTattooMask=NamedRerouteUsage()
    MaterialAttributes=StaticSwitchParameter(True=@n8.Result, False=@n7) {group=GoldCatClaw, parameterName=UseFlipbookMouth}
   ) {materialFunction=/BRCosmetics/Characters/Player/Male/Large/Bodies/M_LRG_GoldCat_Claw/Materials/MaterialFunctions/MF_GoldCat_Reactivity}.Result) {name=OutMaterialAttributes, nodeColor=<rgba 0.5469 1 0 1>}
  NamedRerouteDeclaration(None=@n1.Tech Art Mask) {name=InFXMask, nodeColor=<rgba 0.1484 1 0 1>}
  NamedRerouteDeclaration(None=@n1.M) {name=InMask, nodeColor=<rgba 1 0 0.9375 1>}
 ]
}
