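# Parent hair material for the GoldCat Claw character, layering hair shading and a character-specific reactivity effect onto the shared 2023 character material.
# PreFX attributes pass through the 2023 hair function, blended in by its HairMask and toggled with a static switch.
# An optional fly-away hair stage, off by default, follows, then GoldCat reactivity feeds the shared PostFX function.
# Material attributes flow throughout, with shading model set by expression.
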Material /BRCosmetics/Characters/Player/Male/Large/Bodies/M_LRG_GoldCat_Claw/Materials/M_GoldCat_HairParent
graph {
 properties {
  bUseMaterialAttributes=true
  bUsedWithInstancedStaticMeshes=true
  bUsedWithMeshDeformer=true
  bUsedWithSkeletalMesh=true
  pixelDepthOffsetMode=PDOM_Legacy
  shadingModel=MSM_FromMaterialExpression
  usedShadingModels=Default Lit | Subsurface | Clear Coat | Subsurface Profile | Cloth
 }
 nodes [
  n1 = MaterialFunctionCall() {materialFunction=&MF_CharacterParent_2023_PreFX}
  n2 = MaterialFunctionCall!(In=@n1.Result, Mask Input=@n1.Result) {materialFunction=/Game/Global/Materials/Character_Parent_2023/Hair_Parent_2023/MaterialFunctions/MF_FN_Hair_2023}
  n3 = BlendMaterialAttributes(
   A=@n1.Result
   B=StaticSwitchParameter(True=@n2.Result, False=@n1.Result) {defaultValue=true, group=Skin FX, parameterName=UseFNHair2_0}
   Alpha=@n2.HairMask
  )
  n4 = MaterialFunctionCall!(MaterialInput=@n3) {materialFunction=/Game/Global/Materials/Character_Parent_2023/Hair_Parent_2023/MaterialFunctions/MF_FlyAwayHair_2023}
  MaterialFunctionCall(
   Material=MaterialFunctionCall!(
    Mask
    MouthFlipbook
    InTattooMask
    MaterialAttributes=StaticSwitchParameter(True=@n4.Result, False=@n3) {group=Skin FX, parameterName=IsFlyAwayHair?}
   ) {materialFunction=/BRCosmetics/Characters/Player/Male/Large/Bodies/M_LRG_GoldCat_Claw/Materials/MaterialFunctions/MF_GoldCat_Reactivity}.Result
   M=@n1.M
   Pre Skinned Local Position=@n1.M
   Distance Blend=@n1.M
   Fresnel=@n1.M
  ) {materialFunction=&MF_CharacterParent_2023_PostFX}
  Reroute(None=@n1.Tech Art Mask)
 ]
}
