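# Post-pose Animation Blueprint for the GoldCat Claw large male player body, adding secondary motion to its accessories.
# It copies the attached parent's pose through a deform Control Rig, then bone constraints on pistol bones.
# A chain of AnimDynamics nodes swings pistols, trinket, grenades, rings, ankle, waist and suspender bones before a final Control Rig.
# Each update retunes trinket, grenade and suspender simulation for the front end, except during emotes or skydiving.
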AnimBlueprint /BRCosmetics/Characters/Player/Male/Large/Bodies/M_LRG_GoldCat_Claw/Meshes/M_LRG_GoldCat_Claw_AnimBP
target_skeleton=&/M_LRG_GoldCat_Claw_Skeleton
graphs [
 AnimGraph {nodes [
   n1 = Get AnimDynamics_SimSpace_Grenade_Frontend() -> (AnimDynamics_SimSpace_Grenade_Frontend:AnimPhysSimSpaceType Enum)
   n2 = Get Anim_Dynamics_Alpha_Grenade_Frontend() -> (Anim_Dynamics_Alpha_Grenade_Frontend:double)
   n3 = Get Anim_Dynamics_Gravity_Grenade_Frontend() -> (Anim_Dynamics_Gravity_Grenade_Frontend:Vector)
   n4 = Control Rig(Source=Copy Pose From Mesh() {bUseAttachedParent=true}) {
    alphaBoolBlend {blendOption=Linear}
    controlRigAssetReference {blueprintRigClass=&/M_LRG_GoldCat_Claw_CtrlRig_Deform.M_LRG_GoldCat_Claw_CtrlRig_Deform_C}
   }
   n5 = "Constraint - dyn_pistol_l_zero (PPPPPP)"!(
    ComponentPose=Local To Component(LocalPose=@n4)
    Alpha=1
    ConstraintWeights_0=0.161
    ConstraintWeights_1:float=0.246
    ConstraintWeights_2:float=0.283
    ConstraintWeights_3:float=0.021
    ConstraintWeights_4:float=0.183
    ConstraintWeights_5:float=0.106
   ) -> (Pose) {
    alphaBoolBlend {blendOption=Linear}
    boneToModify {boneName=dyn_pistol_l_zero}
    constraintSetup [
     {targetBone {boneName=spine_02}}
     {targetBone {boneName=spine_03}}
     {targetBone {boneName=spine_04}}
     {targetBone {boneName=spine_01}}
     {targetBone {boneName=deform_pec_l}}
     {targetBone {boneName=clavicle_l}}
    ]
    constraintWeights [1,1,1,1,1,1]
   }
   n6 = "Constraint - dyn_pistol_r (PPPPPP)"!(
    ComponentPose=@n5
    Alpha=1
    ConstraintWeights_0=0.161
    ConstraintWeights_1:float=0.246
    ConstraintWeights_2:float=0.283
    ConstraintWeights_3:float=0.021
    ConstraintWeights_4:float=0.183
    ConstraintWeights_5:float=0.106
   ) -> (Pose) {
    alphaBoolBlend {blendOption=Linear}
    boneToModify {boneName=dyn_pistol_r}
    constraintSetup [
     {targetBone {boneName=spine_02}}
     {targetBone {boneName=spine_03}}
     {targetBone {boneName=spine_04}}
     {targetBone {boneName=spine_01}}
     {targetBone {boneName=deform_pec_r}}
     {targetBone {boneName=clavicle_r}}
    ]
    constraintWeights [1,1,1,1,1,1]
   }
   n7 = Anim Dynamics - Bone: dyn_pistol_l(ComponentPose=@n6, Alpha=1, ExternalForce="0,0,0") {
    angularDampingOverride=1
    angularSpringConstant=1500
    bAngularSpring=true
    bGravityOverrideInSimSpace=true
    bOverrideAngularDamping=true
    gravityScale=0.35
    simulationSpace=RootRelative
    alphaBoolBlend {blendOption=Linear}
    alphaScaleBiasClamp {scale=0.02}
    boundBone {boneName=dyn_pistol_l}
    physicsBodyDefinitions [{
      sphereCollisionRadius=0
      boundBone {boneName=dyn_pistol_l}
      boxExtents=<10 20 20>
      constraintSetup {angularLimitsMax=<0 10 0>, angularLimitsMin=<0 -10 -20>, angularTarget=<0 0 0>}
      localJointOffset=<10 0 0>
     }]
   }
   n8 = Anim Dynamics - Bone: dyn_pistol_r(ComponentPose=@n7, Alpha=1, ExternalForce="0,0,0") {
    angularDampingOverride=1
    angularSpringConstant=1500
    bAngularSpring=true
    bGravityOverrideInSimSpace=true
    bOverrideAngularDamping=true
    gravityScale=0.35
    simulationSpace=RootRelative
    alphaBoolBlend {blendOption=Linear}
    alphaScaleBiasClamp {scale=0.02}
    boundBone {boneName=dyn_pistol_r}
    physicsBodyDefinitions [{
      sphereCollisionRadius=0
      boundBone {boneName=dyn_pistol_r}
      boxExtents=<10 20 20>
      constraintSetup {angularLimitsMax=<0 10 0>, angularLimitsMin=<0 -10 -20>, angularTarget=<0 0 0>}
      localJointOffset=<10 0 0>
     }]
   }
   n9 = Anim Dynamics - Bone: dyn_pistol_trinket_01!(
    SimulationSpace:AnimPhysSimSpaceType Enum=Get AnimDynamics_SimSpace_Trinket_Frontend() -> (AnimDynamics_SimSpace_Trinket_Frontend:AnimPhysSimSpaceType Enum)
    ComponentPose=@n8
    Alpha=Get Anim_Dynamics_Alpha_Trinket_Frontend() -> (Anim_Dynamics_Alpha_Trinket_Frontend:double)
    ExternalForce="0,0,0"
   ) -> (Pose) {
    angularDampingOverride=1
    angularSpringConstant=1500
    bAngularSpring=true
    bGravityOverrideInSimSpace=true
    bOverrideAngularDamping=true
    gravityScale=0.35
    simulationSpace=RootRelative
    alphaBoolBlend {blendOption=Linear}
    alphaScaleBiasClamp {scale=0.35}
    boundBone {boneName=dyn_pistol_trinket_01}
    physicsBodyDefinitions [{
      sphereCollisionRadius=0
      boundBone {boneName=dyn_pistol_trinket_01}
      boxExtents=<5 5 5>
      constraintSetup {angularLimitsMax=<20 30 120>, angularLimitsMin=<-20 -30 0>, angularTarget=<0 0 0>}
      localJointOffset=<5 0 0>
     }]
    relativeSpaceBone {boneName=spine_04}
   }
   n10 = Anim Dynamics - Bone: dyn_grenade_1!(SimulationSpace:AnimPhysSimSpaceType Enum=@n1, ComponentPose=@n9, Alpha=@n2, GravityOverride:Vector=@n3, ExternalForce="0,0,0") -> (Pose) {
    angularDampingOverride=1
    angularSpringConstant=1500
    bAngularSpring=true
    bGravityOverrideInSimSpace=true
    bUseGravityOverride=true
    gravityScale=0.35
    simulationSpace=BoneRelative
    alphaBoolBlend {blendOption=Linear}
    alphaScaleBiasClamp {scale=0.67}
    boundBone {boneName=dyn_grenade_1}
    componentAppliedLinearAccClamp=<5000 5000 5000>
    componentLinearAccScale=<0.37 0.37 0.29>
    componentLinearVelScale=<0.37 0.37 0.29>
    gravityOverride {x=-980}
    physicsBodyDefinitions [{
      sphereCollisionRadius=0
      boundBone {boneName=dyn_grenade_1}
      boxExtents=<9 17 17>
      constraintSetup {angularLimitsMax=<0 0 0>, angularLimitsMin=<0 0 -37>, angularTarget=<0 0 0>}
      localJointOffset=<11 0 0>
     }]
    relativeSpaceBone {boneName=pelvis}
   }
   n11 = Anim Dynamics - Bone: dyn_grenade_2!(SimulationSpace:AnimPhysSimSpaceType Enum=@n1, ComponentPose=@n10, Alpha=@n2, GravityOverride:Vector=@n3, ExternalForce="0,0,0") -> (Pose) {
    angularDampingOverride=1
    angularSpringConstant=1500
    bAngularSpring=true
    bGravityOverrideInSimSpace=true
    bUseGravityOverride=true
    gravityScale=0.35
    simulationSpace=BoneRelative
    alphaBoolBlend {blendOption=Linear}
    alphaScaleBiasClamp {scale=0.72}
    boundBone {boneName=dyn_grenade_2}
    componentAppliedLinearAccClamp=<5000 5000 5000>
    componentLinearAccScale=<0.41 0.41 0.31>
    componentLinearVelScale=<0.41 0.41 0.31>
    gravityOverride {x=-980}
    physicsBodyDefinitions [{
      sphereCollisionRadius=0
      boundBone {boneName=dyn_grenade_2}
      boxExtents=<7 15 15>
      constraintSetup {angularLimitsMax=<0 0 0>, angularLimitsMin=<0 0 -42>, angularTarget=<0 0 0>}
      localJointOffset=<9 0 0>
     }]
    relativeSpaceBone {boneName=pelvis}
   }
   n12 = Anim Dynamics - Bone: dyn_grenade_ring_1!(SimulationSpace:AnimPhysSimSpaceType Enum=BoneRelative, ComponentPose=@n11, Alpha=1, GravityOverride:Vector="490,0,0", ExternalForce="0,0,0") -> (Pose) {
    angularDampingOverride=1
    angularSpringConstant=1500
    bAngularSpring=true
    bGravityOverrideInSimSpace=true
    bOverrideAngularDamping=true
    bUseGravityOverride=true
    gravityScale=0.35
    simulationSpace=BoneRelative
    alphaBoolBlend {blendOption=Linear}
    alphaScaleBiasClamp {scale=0.8}
    boundBone {boneName=dyn_grenade_ring_1}
    componentAppliedLinearAccClamp=<5000 5000 5000>
    componentLinearAccScale=<0.08 0.08 0.08>
    componentLinearVelScale=<0.04 0.04 0.03>
    gravityOverride {x=490}
    physicsBodyDefinitions [{
      sphereCollisionRadius=0
      boundBone {boneName=dyn_grenade_ring_1}
      boxExtents=<7 10 10>
      constraintSetup {angularLimitsMax=<0 0 0>, angularLimitsMin=<0 0 -90>, angularTarget=<0 0 0>}
      localJointOffset=<8 0 0>
     }]
    relativeSpaceBone {boneName=dyn_grenade_ring_1_zero}
   }
   n13 = Anim Dynamics - Bone: dyn_grenade_ring_2!(SimulationSpace:AnimPhysSimSpaceType Enum=BoneRelative, ComponentPose=@n12, Alpha=1, GravityOverride:Vector="490,0,0", ExternalForce="0,0,0") -> (Pose) {
    angularDampingOverride=1
    angularSpringConstant=1500
    bAngularSpring=true
    bGravityOverrideInSimSpace=true
    bOverrideAngularDamping=true
    bUseGravityOverride=true
    gravityScale=0.35
    simulationSpace=BoneRelative
    alphaBoolBlend {blendOption=Linear}
    alphaScaleBiasClamp {scale=0.8}
    boundBone {boneName=dyn_grenade_ring_2}
    componentAppliedLinearAccClamp=<5000 5000 5000>
    componentLinearAccScale=<0.08 0.08 0.08>
    componentLinearVelScale=<0.04 0.04 0.03>
    gravityOverride {x=490}
    physicsBodyDefinitions [{
      sphereCollisionRadius=0
      boundBone {boneName=dyn_grenade_ring_2}
      boxExtents=<7 10 10>
      constraintSetup {angularLimitsMax=<0 0 0>, angularLimitsMin=<0 0 -90>, angularTarget=<0 0 0>}
      localJointOffset=<8 0 0>
     }]
    relativeSpaceBone {boneName=dyn_grenade_ring_2_zero}
   }
   n14 = Anim Dynamics - Bone: dyn_ankle_back_l(ComponentPose=@n13, Alpha=1, ExternalForce="0,0,0") {
    angularDampingOverride=1
    angularSpringConstant=1500
    bAngularSpring=true
    bGravityOverrideInSimSpace=true
    bOverrideAngularDamping=true
    gravityScale=0.35
    simulationSpace=RootRelative
    alphaBoolBlend {blendOption=Linear}
    alphaScaleBiasClamp {scale=0.7}
    boundBone {boneName=dyn_ankle_back_l}
    physicsBodyDefinitions [{
      sphereCollisionRadius=0
      boundBone {boneName=dyn_ankle_back_l}
      boxExtents=<10 20 20>
      constraintSetup {angularLimitsMax=<0 15 0>, angularLimitsMin=<0 -15 -30>, angularTarget=<0 0 0>}
      localJointOffset=<10 0 0>
     }]
   }
   n15 = Anim Dynamics - Bone: dyn_ankle_in_l(ComponentPose=@n14, Alpha=1, ExternalForce="0,0,0") {
    angularDampingOverride=1
    angularSpringConstant=1500
    bAngularSpring=true
    bGravityOverrideInSimSpace=true
    bOverrideAngularDamping=true
    gravityScale=0.35
    simulationSpace=RootRelative
    alphaBoolBlend {blendOption=Linear}
    alphaScaleBiasClamp {scale=0.7}
    boundBone {boneName=dyn_ankle_in_l}
    physicsBodyDefinitions [{
      sphereCollisionRadius=0
      boundBone {boneName=dyn_ankle_in_l}
      boxExtents=<10 20 20>
      constraintSetup {angularLimitsMax=<0 15 0>, angularLimitsMin=<0 -15 -30>, angularTarget=<0 0 0>}
      localJointOffset=<10 0 0>
     }]
   }
   n16 = Anim Dynamics - Bone: dyn_ankle_out_l(ComponentPose=@n15, Alpha=1, ExternalForce="0,0,0") {
    angularDampingOverride=1
    angularSpringConstant=1500
    bAngularSpring=true
    bGravityOverrideInSimSpace=true
    bOverrideAngularDamping=true
    gravityScale=0.35
    simulationSpace=RootRelative
    alphaBoolBlend {blendOption=Linear}
    alphaScaleBiasClamp {scale=0.7}
    boundBone {boneName=dyn_ankle_out_l}
    physicsBodyDefinitions [{
      sphereCollisionRadius=0
      boundBone {boneName=dyn_ankle_out_l}
      boxExtents=<10 20 20>
      constraintSetup {angularLimitsMax=<0 15 0>, angularLimitsMin=<0 -15 -30>, angularTarget=<0 0 0>}
      localJointOffset=<10 0 0>
     }]
   }
   n17 = Anim Dynamics - Bone: dyn_ankle_front_l(ComponentPose=@n16, Alpha=1, ExternalForce="0,0,0") {
    angularDampingOverride=1
    angularSpringConstant=1500
    bAngularSpring=true
    bGravityOverrideInSimSpace=true
    bOverrideAngularDamping=true
    gravityScale=0.35
    simulationSpace=RootRelative
    alphaBoolBlend {blendOption=Linear}
    alphaScaleBiasClamp {scale=0.7}
    boundBone {boneName=dyn_ankle_front_l}
    physicsBodyDefinitions [{
      sphereCollisionRadius=0
      boundBone {boneName=dyn_ankle_front_l}
      boxExtents=<10 20 20>
      constraintSetup {angularLimitsMax=<0 15 0>, angularLimitsMin=<0 -15 -30>, angularTarget=<0 0 0>}
      localJointOffset=<10 0 0>
     }]
   }
   n18 = Anim Dynamics - Bone: dyn_ankle_back_r(ComponentPose=@n17, Alpha=1, ExternalForce="0,0,0") {
    angularDampingOverride=1
    angularSpringConstant=1500
    bAngularSpring=true
    bGravityOverrideInSimSpace=true
    bOverrideAngularDamping=true
    gravityScale=0.35
    simulationSpace=RootRelative
    alphaBoolBlend {blendOption=Linear}
    alphaScaleBiasClamp {scale=0.7}
    boundBone {boneName=dyn_ankle_back_r}
    physicsBodyDefinitions [{
      sphereCollisionRadius=0
      boundBone {boneName=dyn_ankle_back_r}
      boxExtents=<10 20 20>
      constraintSetup {angularLimitsMax=<0 15 0>, angularLimitsMin=<0 -15 -30>, angularTarget=<0 0 0>}
      localJointOffset=<10 0 0>
     }]
   }
   n19 = Anim Dynamics - Bone: dyn_ankle_in_r(ComponentPose=@n18, Alpha=1, ExternalForce="0,0,0") {
    angularDampingOverride=1
    angularSpringConstant=1500
    bAngularSpring=true
    bGravityOverrideInSimSpace=true
    bOverrideAngularDamping=true
    gravityScale=0.35
    simulationSpace=RootRelative
    alphaBoolBlend {blendOption=Linear}
    alphaScaleBiasClamp {scale=0.7}
    boundBone {boneName=dyn_ankle_in_r}
    physicsBodyDefinitions [{
      sphereCollisionRadius=0
      boundBone {boneName=dyn_ankle_in_r}
      boxExtents=<10 20 20>
      constraintSetup {angularLimitsMax=<0 15 0>, angularLimitsMin=<0 -15 -30>, angularTarget=<0 0 0>}
      localJointOffset=<10 0 0>
     }]
   }
   n20 = Anim Dynamics - Bone: dyn_ankle_out_r(ComponentPose=@n19, Alpha=1, ExternalForce="0,0,0") {
    angularDampingOverride=1
    angularSpringConstant=1500
    bAngularSpring=true
    bGravityOverrideInSimSpace=true
    bOverrideAngularDamping=true
    gravityScale=0.35
    simulationSpace=RootRelative
    alphaBoolBlend {blendOption=Linear}
    alphaScaleBiasClamp {scale=0.7}
    boundBone {boneName=dyn_ankle_out_r}
    physicsBodyDefinitions [{
      sphereCollisionRadius=0
      boundBone {boneName=dyn_ankle_out_r}
      boxExtents=<10 20 20>
      constraintSetup {angularLimitsMax=<0 20 0>, angularLimitsMin=<0 -20 -40>, angularTarget=<0 0 0>}
      localJointOffset=<10 0 0>
     }]
   }
   n21 = Anim Dynamics - Bone: dyn_ankle_front_r(ComponentPose=@n20, Alpha=1, ExternalForce="0,0,0") {
    angularDampingOverride=1
    angularSpringConstant=1500
    bAngularSpring=true
    bGravityOverrideInSimSpace=true
    bOverrideAngularDamping=true
    gravityScale=0.35
    simulationSpace=RootRelative
    alphaBoolBlend {blendOption=Linear}
    alphaScaleBiasClamp {scale=0.7}
    boundBone {boneName=dyn_ankle_front_r}
    physicsBodyDefinitions [{
      sphereCollisionRadius=0
      boundBone {boneName=dyn_ankle_front_r}
      boxExtents=<10 20 20>
      constraintSetup {angularLimitsMax=<0 20 0>, angularLimitsMin=<0 -20 -40>, angularTarget=<0 0 0>}
      localJointOffset=<10 0 0>
     }]
   }
   n22 = Anim Dynamics - Bone: dyn_waist_r(ComponentPose=@n21, Alpha=1, ExternalForce="0,0,0") {
    angularDampingOverride=1
    angularSpringConstant=1500
    bAngularSpring=true
    bGravityOverrideInSimSpace=true
    bOverrideAngularDamping=true
    gravityScale=0.35
    simulationSpace=RootRelative
    alphaBoolBlend {blendOption=Linear}
    alphaScaleBiasClamp {scale=0.4}
    boundBone {boneName=dyn_waist_r}
    physicsBodyDefinitions [{
      sphereCollisionRadius=0
      boundBone {boneName=dyn_waist_r}
      boxExtents=<10 20 20>
      constraintSetup {angularLimitsMax=<0 20 0>, angularLimitsMin=<0 -20 -40>, angularTarget=<0 0 0>}
      localJointOffset=<10 0 0>
     }]
   }
   n23 = Anim Dynamics - Bone: dyn_waist_l(ComponentPose=@n22, Alpha=1, ExternalForce="0,0,0") {
    angularDampingOverride=1
    angularSpringConstant=1500
    bAngularSpring=true
    bGravityOverrideInSimSpace=true
    bOverrideAngularDamping=true
    gravityScale=0.35
    simulationSpace=RootRelative
    alphaBoolBlend {blendOption=Linear}
    alphaScaleBiasClamp {scale=0.4}
    boundBone {boneName=dyn_waist_l}
    physicsBodyDefinitions [{
      sphereCollisionRadius=0
      boundBone {boneName=dyn_waist_l}
      boxExtents=<10 20 20>
      constraintSetup {angularLimitsMax=<0 20 0>, angularLimitsMin=<0 -20 -40>, angularTarget=<0 0 0>}
      localJointOffset=<10 0 0>
     }]
   }
   n24 = Anim Dynamics - Bone: dyn_waist_bck_r(ComponentPose=@n23, Alpha=1, ExternalForce="0,0,0") {
    angularDampingOverride=1
    angularSpringConstant=1500
    bAngularSpring=true
    bGravityOverrideInSimSpace=true
    bOverrideAngularDamping=true
    gravityScale=0.35
    simulationSpace=RootRelative
    alphaBoolBlend {blendOption=Linear}
    alphaScaleBiasClamp {scale=0.4}
    boundBone {boneName=dyn_waist_bck_r}
    componentAppliedLinearAccClamp {z=10000}
    componentLinearAccScale {z=0.46}
    componentLinearVelScale {z=0.46}
    physicsBodyDefinitions [{
      sphereCollisionRadius=0
      boundBone {boneName=dyn_waist_bck_r}
      boxExtents=<10 20 20>
      constraintSetup {angularLimitsMax=<0 20 0>, angularLimitsMin=<0 -20 -40>, angularTarget=<0 0 0>}
      localJointOffset=<10 0 0>
     }]
   }
   n25 = Anim Dynamics - Bone: dyn_waist_bck_l(ComponentPose=@n24, Alpha=1, ExternalForce="0,0,0") {
    angularDampingOverride=1
    angularSpringConstant=1500
    bAngularSpring=true
    bGravityOverrideInSimSpace=true
    bOverrideAngularDamping=true
    gravityScale=0.35
    simulationSpace=RootRelative
    alphaBoolBlend {blendOption=Linear}
    alphaScaleBiasClamp {scale=0.4}
    boundBone {boneName=dyn_waist_bck_l}
    componentAppliedLinearAccClamp {z=10000}
    componentLinearAccScale {z=0.46}
    componentLinearVelScale {z=0.46}
    physicsBodyDefinitions [{
      sphereCollisionRadius=0
      boundBone {boneName=dyn_waist_bck_l}
      boxExtents=<10 20 20>
      constraintSetup {angularLimitsMax=<0 20 0>, angularLimitsMin=<0 -20 -40>, angularTarget=<0 0 0>}
      localJointOffset=<10 0 0>
     }]
   }
   n26 = Anim Dynamics - Bone: dyn_waist_2_r(ComponentPose=@n25, Alpha=1, ExternalForce="0,0,0") {
    angularDampingOverride=1
    angularSpringConstant=1500
    bAngularSpring=true
    bGravityOverrideInSimSpace=true
    bOverrideAngularDamping=true
    gravityScale=0.35
    simulationSpace=RootRelative
    alphaBoolBlend {blendOption=Linear}
    alphaScaleBiasClamp {scale=0.4}
    boundBone {boneName=dyn_waist_2_r}
    componentAppliedLinearAccClamp {z=10000}
    componentLinearAccScale {z=0.46}
    componentLinearVelScale {z=0.46}
    physicsBodyDefinitions [{
      sphereCollisionRadius=0
      boundBone {boneName=dyn_waist_2_r}
      boxExtents=<10 20 20>
      constraintSetup {angularLimitsMax=<0 20 0>, angularLimitsMin=<0 -20 -40>, angularTarget=<0 0 0>}
      localJointOffset=<10 0 0>
     }]
   }
   n27 = Anim Dynamics - Bone: dyn_waist_1_r(ComponentPose=@n26, Alpha=1, ExternalForce="0,0,0") {
    angularDampingOverride=1
    angularSpringConstant=1500
    bAngularSpring=true
    bGravityOverrideInSimSpace=true
    bOverrideAngularDamping=true
    gravityScale=0.35
    simulationSpace=RootRelative
    alphaBoolBlend {blendOption=Linear}
    alphaScaleBiasClamp {scale=0.4}
    boundBone {boneName=dyn_waist_1_r}
    componentAppliedLinearAccClamp {z=10000}
    componentLinearAccScale {z=0.46}
    componentLinearVelScale {z=0.46}
    physicsBodyDefinitions [{
      sphereCollisionRadius=0
      boundBone {boneName=dyn_waist_1_r}
      boxExtents=<10 20 20>
      constraintSetup {angularLimitsMax=<0 20 0>, angularLimitsMin=<0 -20 -40>, angularTarget=<0 0 0>}
      localJointOffset=<10 0 0>
     }]
   }
   n28 = Anim Dynamics - Bone: dyn_suspender_l!(
    SimulationSpace:AnimPhysSimSpaceType Enum=Get AnimDynamics_SimSpace_Suspender_Frontend() -> (AnimDynamics_SimSpace_Suspender_Frontend:AnimPhysSimSpaceType Enum)
    ComponentPose=@n27
    Alpha=Get Anim_Dynamics_Alpha_Suspender_Frontend() -> (Anim_Dynamics_Alpha_Suspender_Frontend:double)
    ExternalForce="0,0,0"
   ) -> (Pose) {
    angularDampingOverride=1
    angularSpringConstant=1500
    bAngularSpring=true
    bGravityOverrideInSimSpace=true
    bOverrideAngularDamping=true
    gravityScale=0.35
    simulationSpace=RootRelative
    alphaBoolBlend {blendOption=Linear}
    alphaScaleBiasClamp {scale=0.6}
    boundBone {boneName=dyn_suspender_l}
    componentAppliedLinearAccClamp {z=5000}
    componentLinearAccScale {z=0.45}
    componentLinearVelScale {z=0.45}
    physicsBodyDefinitions [{
      sphereCollisionRadius=0
      boundBone {boneName=dyn_suspender_l}
      boxExtents=<7 16 16>
      constraintSetup {angularLimitsMax=<0 12 0>, angularLimitsMin=<0 -12 -45>, angularTarget=<0 0 0>}
      localJointOffset=<7 0 0>
     }]
   }
   n29 = Control Rig!(
    Source=Component To Local(ComponentPose=@n28)
    Alpha
    Trinket_FE_Alpha:double=Get Trinket_FE_Alpha_ControlRig() -> (Trinket_FE_Alpha_ControlRig:double)
   ) -> (Pose) {
    alphaBoolBlend {blendOption=Linear}
    controlRigAssetReference {blueprintRigClass=&/M_LRG_GoldCat_Claw_CtrlRig.M_LRG_GoldCat_Claw_CtrlRig_C}
   }
   Output Pose(Result=@n29) {layerGroup=None}
  ]}
 EventGraph {nodes [
   Event BlueprintUpdateAnimation()
   n1 = NOT Boolean(A=Get bIsPlayingEmote())
   n2 = NOT Boolean(A=Get bIsSkyDiving())
   > n3 = Branch(Condition=AND Boolean(A=Get bIsInFrontEnd(), B=AND Boolean(A=@n2, B=@n1)))
   > Set AnimDynamics_SimSpace_Trinket_Frontend(AnimDynamics_SimSpace_Trinket_Frontend:AnimPhysSimSpaceType Enum=World) -> (Output_Get:AnimPhysSimSpaceType Enum)
   > Set Anim_Dynamics_Alpha_Trinket_Frontend(Anim_Dynamics_Alpha_Trinket_Frontend:double=0.4) -> (Output_Get:double)
   > Set Trinket_FE_Alpha_ControlRig(Trinket_FE_Alpha_ControlRig:double=0) -> (Output_Get:double)
   > Set AnimDynamics_SimSpace_Grenade_Frontend(AnimDynamics_SimSpace_Grenade_Frontend:AnimPhysSimSpaceType Enum=World) -> (Output_Get:AnimPhysSimSpaceType Enum)
   > Set Anim_Dynamics_Alpha_Grenade_Frontend(Anim_Dynamics_Alpha_Grenade_Frontend:double=0.2) -> (Output_Get:double)
   > Set Anim_Dynamics_Gravity_Grenade_Frontend(Anim_Dynamics_Gravity_Grenade_Frontend:Vector="0,0,-980") -> (Output_Get:Vector)
   > Set AnimDynamics_SimSpace_Suspender_Frontend(AnimDynamics_SimSpace_Suspender_Frontend:AnimPhysSimSpaceType Enum=World) -> (Output_Get:AnimPhysSimSpaceType Enum)
   > Set Anim_Dynamics_Alpha_Suspender_Frontend(Anim_Dynamics_Alpha_Suspender_Frontend:double=0.3) -> (Output_Get:double)
   Set AnimDynamics_SimSpace_Trinket_Frontend(execute=@n3.else)
   > Set Anim_Dynamics_Alpha_Trinket_Frontend(Anim_Dynamics_Alpha_Trinket_Frontend=1)
   > Set Trinket_FE_Alpha_ControlRig(Trinket_FE_Alpha_ControlRig=1)
   > Set AnimDynamics_SimSpace_Grenade_Frontend(AnimDynamics_SimSpace_Grenade_Frontend=BoneRelative)
   > Set Anim_Dynamics_Alpha_Grenade_Frontend(Anim_Dynamics_Alpha_Grenade_Frontend=1)
   > Set Anim_Dynamics_Gravity_Grenade_Frontend(Anim_Dynamics_Gravity_Grenade_Frontend="-980,0,0")
   > Set AnimDynamics_SimSpace_Suspender_Frontend(AnimDynamics_SimSpace_Suspender_Frontend=Component)
   > Set Anim_Dynamics_Alpha_Suspender_Frontend(Anim_Dynamics_Alpha_Suspender_Frontend=1)
  ]}
]
variables [
 Anim_Dynamics_Alpha_Grenade_Frontend {type=number}
 Anim_Dynamics_Alpha_Suspender_Frontend {type=number}
 Anim_Dynamics_Alpha_Trinket_Frontend {type=number}
 Anim_Dynamics_Gravity_Grenade_Frontend {type=Vector}
 AnimDynamics_SimSpace_Grenade_Frontend {type=AnimPhysSimSpaceType}
 AnimDynamics_SimSpace_Suspender_Frontend {type=AnimPhysSimSpaceType}
 AnimDynamics_SimSpace_Trinket_Frontend {type=AnimPhysSimSpaceType}
 Trinket_FE_Alpha_ControlRig {type=number}
]
