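# Physics asset for the large male GoldCat Claw player body, named as a cape collider, giving a body volume for cape collision.
# One capsule per major bone covers legs, feet, pelvis, spine, neck, clavicles, arms and hands, with a single sphere on the head.
# Constraints link neighboring bones into a skeleton chain with all swing and twist motion free, so the listed angle limits do not restrict movement.
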PhysicsAsset /BRCosmetics/Characters/Player/Male/Large/Bodies/M_LRG_GoldCat_Claw/Meshes/M_LRG_GoldCat_Claw_CapeCollider
bodies {
 calf_l [{shapeType=Capsule, length=38.2556, radius=8.1912, shapeName=calf_l_capsule, center=<-18.0265 0.0001 0>, rotation=<rot 90 160.5288 160.529>}]
 calf_r [{shapeType=Capsule, length=38.2556, radius=8.1912, shapeName=calf_r_capsule, center=<17.9789 -0.0001 0>, rotation=<rot 90 160.5288 160.529>}]
 clavicle_l [{
   shapeType=Capsule
   length=20.6258
   radius=15.6338
   shapeName=clavicle_l_capsule
   center=<8.1158 0 0.1473>
   rotation=<rot -80 179.9989 -179.999>
  }]
 clavicle_r [{
   shapeType=Capsule
   length=20.6258
   radius=15.6338
   shapeName=clavicle_r_capsule
   center=<-10.1888 0 -0.027>
   rotation=<rot -80 179.9989 -179.9989>
  }]
 foot_l [{
   shapeType=Capsule
   length=12.0711
   radius=8.5
   shapeName=foot_l_capsule
   center=<-6.5073 7.5993 0.7544>
   rotation=<rot 75.894 -134.5617 -45.4383>
  }]
 foot_r [{
   shapeType=Capsule
   length=7.5711
   radius=9
   shapeName=foot_r_capsule
   center=<7.1979 -7.0315 -1.1398>
   rotation=<rot 75.8941 -134.5618 -45.4381>
  }]
 hand_l [{shapeType=Capsule, length=11, radius=6.5, shapeName=hand_l_capsule, center=<6.7781 -1.0794 0>, rotation=<rot 90 -179.9603 -179.9607>}]
 hand_r [{shapeType=Capsule, length=10.5, radius=7, shapeName=hand_r_capsule, center=<-5.1681 2.1126 0>, rotation=<rot -90 0 0>}]
 head [{shapeType=Sphere, radius=15.0178, shapeName=head_sphere, center=<10.0448 0 0>}]
 lowerarm_l [{shapeType=Capsule, length=18, radius=8, shapeName=lowerarm_l_capsule, center=<6.1765 0 0>, rotation=<rot 90 -0.0007 -0.001>}]
 lowerarm_r [{shapeType=Capsule, length=18, radius=8, shapeName=lowerarm_r_capsule, center=<-5.2744 0 0>, rotation=<rot 90 -160.5289 -160.5293>}]
 neck_02 [{shapeType=Capsule, length=4, radius=6, shapeName=neck_02_capsule, rotation=<rot -90 -13.5122 13.5121>}]
 pelvis [{shapeType=Capsule, length=27.5978, radius=12.9116, shapeName=pelvis_capsule, center=<-7.5341 0 0>}]
 spine_01 [{shapeType=Capsule, length=16.2275, radius=14.6316, shapeName=spine_01_capsule}]
 spine_02 [{shapeType=Capsule, length=24.0999, radius=13.416, shapeName=spine_02_capsule}]
 spine_03 [{shapeType=Capsule, length=26.6114, radius=16.0069, shapeName=spine_03_capsule}]
 spine_04 [{shapeType=Capsule, length=24.0437, radius=16.2263, shapeName=spine_04_capsule}]
 spine_05 [{shapeType=Capsule, length=28.5526, radius=16.4077, shapeName=spine_05_capsule, center=<-2.9343 -1.2899 0>}]
 thigh_l [{shapeType=Capsule, length=31.2076, radius=9.3691, shapeName=thigh_l_capsule, center=<-21.7647 0.0002 0>, rotation=<rot 90 3.0357 3.0362>}]
 thigh_r [{shapeType=Capsule, length=31.2076, radius=9.3691, shapeName=thigh_r_capsule, center=<22.3523 -0.0002 0>, rotation=<rot 90 179.9602 179.9608>}]
 upperarm_l [{shapeType=Capsule, length=14.8682, radius=11.6318, shapeName=upperarm_l_capsule, center=<3.8841 0 0>, rotation=<rot 80 -0.0002 -0.0008>}]
 upperarm_r [{shapeType=Capsule, length=14.8682, radius=11.6318, shapeName=upperarm_r_capsule, center=<-2.3175 0 0>, rotation=<rot 80 -0.0002 -0.0008>}]
}
constraints [
 {
  bone1Name=hand_l
  bone2Name=lowerarm_l
  swing1LimitDegrees=45
  swing1Motion=Free
  swing2LimitDegrees=45
  swing2Motion=Free
  twistLimitDegrees=45
  twistMotion=Free
 }
 {
  bone1Name=calf_l
  bone2Name=thigh_l
  swing1LimitDegrees=45
  swing1Motion=Free
  swing2LimitDegrees=45
  swing2Motion=Free
  twistLimitDegrees=45
  twistMotion=Free
 }
 {
  bone1Name=foot_l
  bone2Name=calf_l
  swing1LimitDegrees=45
  swing1Motion=Free
  swing2LimitDegrees=45
  swing2Motion=Free
  twistLimitDegrees=45
  twistMotion=Free
 }
 {
  bone1Name=calf_r
  bone2Name=thigh_r
  swing1LimitDegrees=45
  swing1Motion=Free
  swing2LimitDegrees=45
  swing2Motion=Free
  twistLimitDegrees=45
  twistMotion=Free
 }
 {
  bone1Name=foot_r
  bone2Name=calf_r
  swing1LimitDegrees=45
  swing1Motion=Free
  swing2LimitDegrees=45
  swing2Motion=Free
  twistLimitDegrees=45
  twistMotion=Free
 }
 {
  bone1Name=thigh_r
  bone2Name=pelvis
  swing1LimitDegrees=45
  swing1Motion=Free
  swing2LimitDegrees=45
  swing2Motion=Free
  twistLimitDegrees=45
  twistMotion=Free
 }
 {
  bone1Name=thigh_l
  bone2Name=pelvis
  swing1LimitDegrees=45
  swing1Motion=Free
  swing2LimitDegrees=45
  swing2Motion=Free
  twistLimitDegrees=45
  twistMotion=Free
 }
 {
  bone1Name=spine_01
  bone2Name=pelvis
  swing1LimitDegrees=45
  swing1Motion=Free
  swing2LimitDegrees=45
  swing2Motion=Free
  twistLimitDegrees=45
  twistMotion=Free
 }
 {
  bone1Name=spine_02
  bone2Name=spine_01
  swing1LimitDegrees=45
  swing1Motion=Free
  swing2LimitDegrees=45
  swing2Motion=Free
  twistLimitDegrees=45
  twistMotion=Free
 }
 {
  bone1Name=spine_03
  bone2Name=spine_02
  swing1LimitDegrees=45
  swing1Motion=Free
  swing2LimitDegrees=45
  swing2Motion=Free
  twistLimitDegrees=45
  twistMotion=Free
 }
 {
  bone1Name=spine_04
  bone2Name=spine_03
  swing1LimitDegrees=45
  swing1Motion=Free
  swing2LimitDegrees=45
  swing2Motion=Free
  twistLimitDegrees=45
  twistMotion=Free
 }
 {
  bone1Name=spine_05
  bone2Name=spine_04
  swing1LimitDegrees=45
  swing1Motion=Free
  swing2LimitDegrees=45
  swing2Motion=Free
  twistLimitDegrees=45
  twistMotion=Free
 }
 {
  bone1Name=neck_02
  bone2Name=head
  swing1LimitDegrees=45
  swing1Motion=Free
  swing2LimitDegrees=45
  swing2Motion=Free
  twistLimitDegrees=45
  twistMotion=Free
 }
 {
  bone1Name=spine_05
  bone2Name=neck_02
  swing1LimitDegrees=45
  swing1Motion=Free
  swing2LimitDegrees=45
  swing2Motion=Free
  twistLimitDegrees=45
  twistMotion=Free
 }
 {
  bone1Name=clavicle_l
  bone2Name=spine_05
  swing1LimitDegrees=45
  swing1Motion=Free
  swing2LimitDegrees=45
  swing2Motion=Free
  twistLimitDegrees=45
  twistMotion=Free
 }
 {
  bone1Name=clavicle_r
  bone2Name=spine_05
  swing1LimitDegrees=45
  swing1Motion=Free
  swing2LimitDegrees=45
  swing2Motion=Free
  twistLimitDegrees=45
  twistMotion=Free
 }
 {
  bone1Name=upperarm_r
  bone2Name=clavicle_r
  swing1LimitDegrees=45
  swing1Motion=Free
  swing2LimitDegrees=45
  swing2Motion=Free
  twistLimitDegrees=45
  twistMotion=Free
 }
 {
  bone1Name=upperarm_l
  bone2Name=clavicle_l
  swing1LimitDegrees=45
  swing1Motion=Free
  swing2LimitDegrees=45
  swing2Motion=Free
  twistLimitDegrees=45
  twistMotion=Free
 }
 {
  bone1Name=lowerarm_r
  bone2Name=upperarm_r
  swing1LimitDegrees=45
  swing1Motion=Free
  swing2LimitDegrees=45
  swing2Motion=Free
  twistLimitDegrees=45
  twistMotion=Free
 }
 {
  bone1Name=hand_r
  bone2Name=lowerarm_r
  swing1LimitDegrees=45
  swing1Motion=Free
  swing2LimitDegrees=45
  swing2Motion=Free
  twistLimitDegrees=45
  twistMotion=Free
 }
 {
  bone1Name=lowerarm_l
  bone2Name=upperarm_l
  swing1LimitDegrees=45
  swing1Motion=Free
  swing2LimitDegrees=45
  swing2Motion=Free
  twistLimitDegrees=45
  twistMotion=Free
 }
]
