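# Post-copy physics AnimBlueprint for the GoldCat Claw male face accessory, adding secondary motion to its dyn_hair bones.
# It copies the attached parent's pose, then chains five angular-spring Anim Dynamics nodes with reduced gravity and per-bone angular limits.
# Each update, front-end play outside emotes and skydiving selects World sim space and reduced alpha; otherwise alpha is full.
# These variables drive the first four nodes; dyn_hair_b stays root-relative at fixed alpha.
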AnimBlueprint /BRCosmetics/Characters/Player/Male/Large/Bodies/M_LRG_GoldCat_Claw/Meshes/Parts/M_LRG_GoldCat_Claw_FaceAcc_AnimBP
target_skeleton=&/M_LRG_GoldCat_Claw_FaceAcc_Skeleton
graphs [
 AnimGraph {nodes [
   n1 = Get AnimDynamics_SimSpace_Frontend() -> (AnimDynamics_SimSpace_Frontend:AnimPhysSimSpaceType Enum)
   n2 = Get Anim_Dynamics_Alpha_Frontend() -> (Anim_Dynamics_Alpha_Frontend:double)
   n3 = Anim Dynamics - Bone: dyn_hair_01!(
    SimulationSpace:AnimPhysSimSpaceType Enum=@n1
    ComponentPose=Local To Component(LocalPose=Copy Pose From Mesh() {bUseAttachedParent=true})
    Alpha=@n2
    ExternalForce
   ) -> (Pose) {
    angularDampingOverride=1
    angularSpringConstant=1500
    bAngularSpring=true
    bGravityOverrideInSimSpace=true
    bOverrideAngularDamping=true
    bUseGravityOverride=true
    gravityScale=0.35
    simulationSpace=RootRelative
    alphaBoolBlend {blendOption=Linear}
    alphaScaleBiasClamp {scale=0.33}
    boundBone {boneName=dyn_hair_01}
    componentAppliedLinearAccClamp {z=5000}
    componentLinearAccScale {z=0.46}
    componentLinearVelScale {z=0.46}
    gravityOverride {z=-980}
    physicsBodyDefinitions [{
      sphereCollisionRadius=0
      boundBone {boneName=dyn_hair_01}
      boxExtents=<10 20 20>
      constraintSetup {angularLimitsMax=<0 0 0>, angularLimitsMin=<0 -20 -25>, angularTarget=<0 0 0>}
      localJointOffset=<10 0 0>
     }]
   }
   n4 = Anim Dynamics - Bone: dyn_hair_02!(SimulationSpace:AnimPhysSimSpaceType Enum=@n1, ComponentPose=@n3, Alpha=@n2, ExternalForce) -> (Pose) {
    angularDampingOverride=1
    angularSpringConstant=1500
    bAngularSpring=true
    bGravityOverrideInSimSpace=true
    bOverrideAngularDamping=true
    bUseGravityOverride=true
    gravityScale=0.35
    simulationSpace=RootRelative
    alphaBoolBlend {blendOption=Linear}
    alphaScaleBiasClamp {scale=0.45}
    boundBone {boneName=dyn_hair_02}
    componentAppliedLinearAccClamp {z=5000}
    componentLinearAccScale {z=0.46}
    componentLinearVelScale {z=0.46}
    gravityOverride {z=-980}
    physicsBodyDefinitions [{
      sphereCollisionRadius=0
      boundBone {boneName=dyn_hair_02}
      boxExtents=<10 20 20>
      constraintSetup {angularLimitsMax=<0 0 0>, angularLimitsMin=<0 -25 -30>, angularTarget=<0 0 0>}
      localJointOffset=<10 0 0>
     }]
   }
   n5 = Anim Dynamics - Bone: dyn_hair_f!(SimulationSpace:AnimPhysSimSpaceType Enum=@n1, ComponentPose=@n4, Alpha=@n2, ExternalForce="0,0,0") -> (Pose) {
    angularDampingOverride=1
    angularSpringConstant=1500
    bAngularSpring=true
    bGravityOverrideInSimSpace=true
    bOverrideAngularDamping=true
    bUseGravityOverride=true
    gravityScale=0.35
    simulationSpace=RootRelative
    alphaBoolBlend {blendOption=Linear}
    alphaScaleBiasClamp {scale=0.45}
    boundBone {boneName=dyn_hair_f}
    componentAppliedLinearAccClamp {z=5000}
    componentLinearAccScale {z=0.46}
    componentLinearVelScale {z=0.46}
    gravityOverride {z=-980}
    physicsBodyDefinitions [{
      sphereCollisionRadius=0
      boundBone {boneName=dyn_hair_f}
      boxExtents=<10 20 20>
      constraintSetup {angularLimitsMax=<0 -3 -15>, angularLimitsMin=<0 -30 -15>, angularTarget=<0 0 0>}
      localJointOffset=<10 0 0>
     }]
   }
   n6 = Anim Dynamics - Bone: dyn_hair_m!(SimulationSpace:AnimPhysSimSpaceType Enum=@n1, ComponentPose=@n5, Alpha=@n2, ExternalForce="0,0,0") -> (Pose) {
    angularDampingOverride=1
    angularSpringConstant=1500
    bAngularSpring=true
    bGravityOverrideInSimSpace=true
    bOverrideAngularDamping=true
    bUseGravityOverride=true
    gravityScale=0.35
    simulationSpace=RootRelative
    alphaBoolBlend {blendOption=Linear}
    alphaScaleBiasClamp {scale=0.42}
    boundBone {boneName=dyn_hair_m}
    componentAppliedLinearAccClamp {z=5000}
    componentLinearAccScale {z=0.46}
    componentLinearVelScale {z=0.46}
    gravityOverride {z=-980}
    physicsBodyDefinitions [{
      sphereCollisionRadius=0
      boundBone {boneName=dyn_hair_m}
      boxExtents=<9 19 19>
      constraintSetup {angularLimitsMax=<0 -5 13>, angularLimitsMin=<0 -28 -13>, angularTarget=<0 0 0>}
      localJointOffset=<9 0 0>
     }]
   }
   n7 = Anim Dynamics - Bone: dyn_hair_b(ComponentPose=@n6, Alpha=1, ExternalForce="0,0,0") {
    angularDampingOverride=1
    angularSpringConstant=1500
    bAngularSpring=true
    bGravityOverrideInSimSpace=true
    bOverrideAngularDamping=true
    bUseGravityOverride=true
    gravityScale=0.35
    simulationSpace=RootRelative
    alphaBoolBlend {blendOption=Linear}
    alphaScaleBiasClamp {scale=0.43}
    boundBone {boneName=dyn_hair_b}
    componentAppliedLinearAccClamp {z=5000}
    componentLinearAccScale {z=0.46}
    componentLinearVelScale {z=0.46}
    gravityOverride {z=-980}
    physicsBodyDefinitions [{
      sphereCollisionRadius=0
      boundBone {boneName=dyn_hair_b}
      boxExtents=<8 18 18>
      constraintSetup {angularLimitsMax=<0 0 3>, angularLimitsMin=<0 -22 -29>, angularTarget=<0 0 0>}
      localJointOffset=<8 0 0>
     }]
   }
   Output Pose(Result=Component To Local(ComponentPose=@n7)) {layerGroup=None}
  ]}
 EventGraph {nodes [
   Event BlueprintUpdateAnimation()
   n1 = NOT Boolean(A=Get bIsPlayingEmote())
   n2 = NOT Boolean(A=Get bIsSkyDiving())
   > n3 = Branch(Condition=AND Boolean(A=Get bIsInFrontEnd(), B=AND Boolean(A=@n2, B=@n1)))
   > Set AnimDynamics_SimSpace_Frontend(AnimDynamics_SimSpace_Frontend:AnimPhysSimSpaceType Enum=World) -> (Output_Get:AnimPhysSimSpaceType Enum)
   > Set Anim_Dynamics_Alpha_Frontend(Anim_Dynamics_Alpha_Frontend:double=0.75) -> (Output_Get:double)
   Set AnimDynamics_SimSpace_Frontend(execute=@n3.else)
   > Set Anim_Dynamics_Alpha_Frontend(Anim_Dynamics_Alpha_Frontend=1)
  ]}
]
variables [
 Anim_Dynamics_Alpha_Frontend {type=number}
 AnimDynamics_SimSpace_Frontend {type=AnimPhysSimSpaceType}
]
